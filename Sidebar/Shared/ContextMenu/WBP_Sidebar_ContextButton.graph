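# Sidebar context menu button, built on the UIKit generic button, that serves as a list entry for one action.
# On receiving its action wrapper, it sets text, enabled and focusable state, async-loaded icon, and clears the hold sound.
# The action's input Interaction applies only when the ContextMenuUI.AllowQuickActions console variable is true.
# Click, press and release forward to the wrapper, while hover drives an outline block that SetOutlineStyle can hide.
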WidgetBlueprint /Sidebar/Shared/ContextMenu/WBP_Sidebar_ContextButton
graphs [
 EventGraph {nodes [
   Event On Clicked()
   > Parent: On Clicked() {is=title K2Node_CallParentFunction}
   > ExecuteInteraction(self:UIKit Context Menu Action Wrapper Object Reference=Get Action() -> (Action:UIKit Context Menu Action Wrapper Object Reference))
   Event On Hovered()
   > Parent: On Hovered() {is=title K2Node_CallParentFunction}
   n1 = Get Outline() -> (Outline:WBP UIKit Block Outline Object Reference)
   > RequestTransition!(execute, self:UIKit Block Object Reference=@n1, State:EUIKitBlockVisualState Enum=Focused) -> (then)
   Event On Unhovered()
   > Parent: On Unhovered() {is=title K2Node_CallParentFunction}
   > RequestTransition(self=@n1, State=Unfocused)
   n2 = Event OnListItemObjectSet()
   > n3 = Cast To UIKitContextMenuActionWrapper!(execute, Object=@n2.ListItemObject) -> (then, CastFailed, AsUIKit Context Menu Action Wrapper:UIKit Context Menu Action Wrapper Object Reference) {is=title K2Node_DynamicCast}
   > Set Action(Action:UIKit Context Menu Action Wrapper Object Reference=@n3) -> (Output_Get:UIKit Context Menu Action Wrapper Object Reference)
   > n4 = Sequence!(execute) -> (then_0, then_1, then_2:Exec)
   n5 = GetConsoleVariableBoolValue!(execute=@n4.then_1, VariableName:String=ContextMenuUI.AllowQuickActions) -> (then, ReturnValue:bool)
   n6 = GetActionInfo!(self:UIKit Context Menu Action Wrapper Object Reference=Get Action()) -> (ReturnValue:"UIKit Context Menu Action Info Structure (by ref)")
   > Set Interaction(Interaction:Data Table Row Handle Structure=Select(
     Option 0:Data Table Row Handle Structure
     Option 1:Data Table Row Handle Structure=Break UIKit Context Menu Action Info!(UIKitContextMenuActionInfo:"UIKit Context Menu Action Info Structure (by ref)"=@n6) -> (
      IconMaterial:Material Instance Soft Object Reference
      bIsSelected:bool=~
      bAddDivider:bool=~
      bPositionDividerAfter:bool=~
      bCloseMenuOnSelect:bool=~
      Interaction:Data Table Row Handle Structure
     ) {is=title K2Node_BreakStruct}.Interaction
     Index=@n5
    ) -> (ReturnValue:Data Table Row Handle Structure)) -> (Output_Get:Data Table Row Handle Structure)
   Set HoldSound(execute=@n4.then_2, HoldSound:Sound Base Object Reference) -> (Output_Get:Sound Base Object Reference)
   n7 = Break UIKit Context Menu Action Info!(UIKitContextMenuActionInfo=GetActionInfo(self=Get Action())) -> (
    DisplayName:Text
    IconTexture:Texture 2D Soft Object Reference
    IconMaterial
    SortPriority:int
    bIsAvailable:bool=~
    bIsEnabled:bool=~
    bIsSelected=~
    bAddDivider=~
    bPositionDividerAfter=~
    bCloseMenuOnSelect=~
    InteractionNameID:Name
    Interaction
    PressedDelegate:Delegate
    ReleasedDelegate:Delegate
   )
   SetText!(execute=@n4.then_0, self:WBP UIKit Button Generic Base Object Reference, Text:Text=@n7) -> (then)
   > SetIsInteractionEnabled!(execute, self:Common Button Base Object Reference, bInIsInteractionEnabled:bool=@n7.bIsEnabled) -> (then)
   > SetIsFocusable!(execute, self:Common Button Base Object Reference, bInIsFocusable:bool=@n7.bIsEnabled) -> (then)
   > n8 = Branch(Condition=IsValidSoftObjectReference(SoftObjectReference=@n7.IconTexture))
   > n9 = Async Load Asset(Asset=@n7.IconTexture)
   SetIcon!(execute=@n9.Completed, self:WBP UIKit Button Generic Base Object Reference, Icon:Object Reference=@n9) -> (then)
   SetIcon(execute=@n8.else)
   Event On Pressed() {is=title K2Node_Event}
   > Parent: On Pressed() {is=title K2Node_CallParentFunction}
   > ExecutePressedDelegate(self:UIKit Context Menu Action Wrapper Object Reference=Get Action())
   Event On Released() {is=title K2Node_Event}
   > Parent: On Released() {is=title K2Node_CallParentFunction}
   > ExecuteReleasedDelegate(self:UIKit Context Menu Action Wrapper Object Reference=Get Action())
  ]}
 SetOutlineStyle {nodes [
   n1 = SetOutlineStyle!() -> (then, Use Outline:bool=~) {is=title K2Node_FunctionEntry}
   > SetVisibility(
    self=Get Outline()
    InVisibility=Select(Option 0:ESlateVisibility Enum=Collapsed, Option 1:ESlateVisibility Enum=SelfHitTestInvisible, Index=@n1) -> (ReturnValue:ESlateVisibility Enum)
   )
  ]}
]
variables [
 Action {type=UIKitContextMenuActionWrapper, default=null}
 Interaction {type=DataTableRowHandle, default {dataTable=null, rowName=None}}
 UseOutline {type=boolean, default=true}
]
widgets {
 info {
  inheritedWidgetCount=0
  namedSlotCount=1
  parentClass=/FortUILibrary/Systems/ModularButton/WBP_UIKit_ButtonGeneric_Base.WBP_UIKit_ButtonGeneric_Base_C
  rootWidgetClass=null
  widgetCount=2
 }
 widgets [
  Overlay_Main {widgetClassPath=&Overlay, bIsVariable=true, widget=Overlay_Main}
  Outline {
   widgetClassPath=/FortUILibrary/UIKit/Buttons/Blocks/WBP_UIKit_Block_Outline
   bIsVariable=true
   parent=Overlay_Main
   widget=Outline
   properties {contain Outline Padding=true, visibility=SelfHitTestInvisible, padding=<ltrb -8 -8 -8 -8>}
   slot {class=OverlaySlot, properties {horizontalAlignment=HAlign_Fill, verticalAlignment=VAlign_Fill}}
  }
 ]
}
